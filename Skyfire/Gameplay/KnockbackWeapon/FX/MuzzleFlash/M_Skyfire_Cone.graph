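# Translucent unlit particle material for the cone-shaped muzzle flash of the Skyfire knockback weapon.
# Normal-map-distorted scrolling noise, offset by actor distance, joins an axis-aligned view-facing mask that softens cone silhouettes.
# A sine along V shapes this, particle alpha and blurred noise erode it, then depth fade, contrast and a flame texture refine opacity.
# MF_BlasterCommon supplies color, while dynamic parameters drive intensity, alpha and a normal-based width offset growing along V.
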Material /Skyfire/Gameplay/KnockbackWeapon/FX/MuzzleFlash/M_Skyfire_Cone
graph {
 properties {
  bUsedWithMeshParticles=true
  bUsedWithNiagaraMeshParticles=true
  bUsedWithParticleSprites=true
  blendMode=BLEND_Translucent
  pixelDepthOffsetMode=PDOM_Legacy
  shadingModel=MSM_Unlit
  usedShadingModels=Unlit
 }
 nodes [
  n1 = MaterialFunctionCall!(TimeScale, ElementColor, HeadEnhanceScale) {materialFunction=/Game/Effects/Fort_Effects/Materials/_Functions/MF_BlasterCommon}
  n2 = TextureCoordinate()
  n3 = SphereMask(A=WorldPosition().XYZ, B=ActorPositionWS()) {attenuationRadius=4096, hardnessPercent=0}
  n4 = Multiply(A=@n3, B=@n3) {constB=2}
  n5 = Multiply(A=Time()) {constB=0.04}
  n6 = Multiply(A=AppendVector(A=ComponentMask(None=@n2) {r=true}, B=Add(A=@n4, B=@n5)), B=Constant2Vector() {g=4, r=2}) {constB=25}
  n7 = TextureSample(UVs=Panner(Coordinate=Multiply(A=@n6) {constB=2}) {speedX=0.4, speedY=0.4}) {samplerType=SAMPLERTYPE_Normal, texture=&T_Weap_Distort_N_02}
  n8 = Add(A=@n6, B=ComponentMask(None=Multiply(A=@n7.RGB) {constB=0.2}) {g=true, r=true})
  n9 = LinearInterpolate(
   A=TextureSample(UVs=@n8) {samplerType=SAMPLERTYPE_LinearColor, texture=&T_NoisePack_LinearColor}.R
   B=TextureSample(UVs=@n8) {texture=/Game/Global/Textures/Default/Noise/TankMoteNoise}.R
  )
  n10 = Normalize(VectorInput=Subtract(A=WorldPosition().XYZ, B=CameraPositionWS()))
  n11 = Transform(None=Constant3Vector() {constant=<rgba 1 0 0 0>}) {transformSourceType=TRANSFORMSOURCE_ParticleWorld}
  n12 = CrossProduct(A=Normalize(VectorInput=CrossProduct(A=@n10, B=@n11)), B=@n11)
  n13 = Abs(None=DotProduct(A=Normalize(VectorInput=@n12), B=VertexNormalWS()))
  n14 = Saturate(None=Multiply(A=Subtract(A=@n13) {constB=0.5}) {constB=4})
  n15 = ComponentMask(None=TextureCoordinate() {uTiling=0.5, vTiling=0.5}) {g=true}
  n16 = TextureSample(UVs=Multiply(A=@n4) {constB=2}) {samplerType=SAMPLERTYPE_Grayscale, texture=&LowResBlurredNoise}
  n17 = Subtract(A=Multiply(A=Add(A=@n9, B=@n14), B=Sine(None=@n15)), B=Add(A=ParticleColor().A, B=Subtract(A=@n16.R) {constB=0.5}))
  n18 = DepthFade() {fadeDistanceDefault=20}
  n19 = Multiply(
   A=MaterialFunctionCall(In=Multiply(A=@n17, B=@n18), Contrast=Constant() {r=0.5}) {materialFunction=&CheapContrast}.Result
   B=@n1.HeadEnhance1
  )
  n20 = Multiply(A=@n19, B=ParticleColor().A)
  Multiply(
   A=Multiply(
    A=StaticSwitchParameter(True=@n20, False=@n19) {parameterName=Multiply Particle Opacity}
    B=DynamicParameter() {parameterIndex=1, defaultValue=<rgba 1 200 1 1>, paramNames [Alpha, "", "", ""]}.Alpha
   )
   B=TextureSample() {samplerType=SAMPLERTYPE_LinearGrayscale, texture=/Game/Weapons/FORT_Melee/Pickaxe_Cube_Ninja_Male/FX/T_FluidFlameTile}.RGB
  )
  n21 = Add(A=Multiply(A=ScalarParameter() {defaultValue=1, parameterName=LerpRange}, B=@n17), B=ScalarParameter() {parameterName=LerpOffset})
  Multiply(
   A=Multiply(
    A=DynamicParameter() {defaultValue=<rgba 1 200 1 1>, paramNames [Random, Width, Intensity, UsedInFunction]}.Intensity
    B=LinearInterpolate(A=@n1.DarkColor, B=@n1.DarkColor, Alpha=@n21)
   )
   B=@n1.HeadEnhance
  )
  n22 = ComponentMask(None=TextureCoordinate()) {g=true}
  Multiply(
   A=VertexNormalWS()
   B=Multiply(
    A=Add(A=@n22) {constB=0.2}
    B=DynamicParameter() {defaultValue=<rgba 1 200 1 1>, paramNames [Random, Width, Intensity, UsedInFunction]}.Width
   )
  )
 ]
}
