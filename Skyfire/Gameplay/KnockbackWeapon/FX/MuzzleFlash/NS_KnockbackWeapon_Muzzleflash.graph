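# Niagara muzzle flash for the knockback weapon: a short cyan-to-magenta energy burst along local +X.
# Layers hot/soft flash meshes, flame and spark sprites, barrel glow with a light, magnetic cylinders, an energy ring and forward-moving rings.
# Age curves drive scale, color and dynamic material parameters.
# User colors tint only some layers; others hardcode theirs, and Color_02 feeds only disabled Core and Streak (MuzzleFlare is also disabled).
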NiagaraSystem /Skyfire/Gameplay/KnockbackWeapon/FX/MuzzleFlash/NS_KnockbackWeapon_Muzzleflash
dependencies {
 usedDataInterfaces [
  &NiagaraDataInterfaceCurve
  &NiagaraDataInterfaceArrayFloat3
  &NiagaraDataInterfaceVectorCurve
  &NiagaraDataInterfaceColorCurve
  &NiagaraDataInterfaceMeshRendererInfo
 ]
 usedDynamicInputs []
 usedModules [&SystemState]
}
emitters {
 BarrelHole {
  emitterUpdateScript [
   &EmitterState()
   &SpawnBurst_Instantaneous(Spawn Count=1, Use Spawn Probability=false)
  ]
  particleSpawnScript [
   /Niagara/Modules/Spawn/Initialization/InitializeParticle(
    Color=$User.Color_01
    Lifetime=&RandomRangeFloat(Minimum=0.4, Maximum=0.45, Evaluation Type=::NewEnumerator1)
    Mesh Scale=&VectorFromFloat(Value=0.35)
    Module.Write Color=true
    Module.Write Lifetime=true
    Module.Write Mass=true
    Module.Write Material Random=true
    Module.Write Position=true
    Module.Write Scale=true
    Module.Write SpriteRotation=false
    Module.Write SpriteSize=true
    Position=&Add_Vector(A=/Niagara/DynamicInputs/Helpers/SimulationPosition(), B=<0 0 0>)
    Sprite Size=&Vector2DFromFloat(Value=8)
   )
   /SetVariables(Particles.SpriteFacing=&RotateVectorByQuaternion(Vector To Rotate=<1 0 0>, Quaternion=$Engine.Owner.Rotation)) {isSetParameters=true}
  ]
  particleUpdateScript [
   &ParticleState()
   &DynamicMaterialParameters(
    Index 0 Param 1=&FloatFromCurve(FloatCurve=DataInterface() {propertyValues {Curve {keys [<at 0 0.75>,<at 1 1.25>]}}})
    Index 0 Param 2=10
    Index 0 Param 3=10
    Index 0 Param 4=10
    Module.Param0WriteEnabled=true
    Module.Param1WriteEnabled=true
    Module.Param2WriteEnabled=true
    Module.Param3WriteEnabled=true
   )
   &ScaleColor(
    Scale Alpha=&FloatFromCurve(
     FloatCurve=DataInterface() {propertyValues {Curve {keys [
         {interpMode=RCIM_Cubic, tangentMode=RCTM_User, time=0, value=0}
         {interpMode=RCIM_Cubic, tangentMode=RCTM_User, time=0.3, value=1}
         {tangentMode=RCTM_User, time=1, value=0}
        ]}}}
     Scale Curve=0.3
    )
    Scale RGB=&VectorFromFloat(Value=200)
    ScaleA=true
    ScaleRGB=true
   )
   &CameraOffset()
   &ScaleSpriteSize(Scale Factor=&Vector2DFromFloat(Value=&FloatFromCurve(FloatCurve=DataInterface() {propertyValues {Curve {keys [{interpMode=RCIM_Cubic, tangentMode=RCTM_User, time=0, value=1}, {interpMode=RCIM_Cubic, tangentMode=RCTM_User, time=1, value=0.5}]}}}))) {enabled=false}
  ]
  rendererClasses [&NiagaraSpriteRendererProperties]
  renderers [{rendererClass=&NiagaraSpriteRendererProperties}]
 }
 Core {
  bEnabled=false
  emitterUpdateScript [
   &EmitterState(Life Cycle Mode=::NewEnumerator1, Loop Behavior=::NewEnumerator1, Loop Duration=0.1, UseLoopDelay=true)
   &SpawnBurst_Instantaneous(Spawn Count=&RandomRangeInt(Minimum=1, Maximum=1), Use Spawn Probability=false)
  ]
  particleSpawnScript [
   /SetVariables(Particles.MeshOrientation=&MakeQuatFromAxisAngle(
     Axis=&RandomRangeVector(Minimum=<-1 0 0>, Maximum=<1 0 0>, Evaluation Type=::NewEnumerator1)
     AngleInDegrees=90
    )) {isSetParameters=true}
   &InitializeParticle(
    Lifetime=0.1
    Mesh Scale=/Niagara/DynamicInputs/CascadeConversion/SplitVector(
     Double Component Vector Channels:ECascadeNiagaraTwoVectorChannels=::NewEnumerator1
     Double Component:NiagaraFloat=&RandomRangeFloat(Minimum=0.25, Maximum=0.3, Evaluation Type=::NewEnumerator1)
     Single Component:NiagaraFloat=&RandomRangeFloat(Minimum=0.3, Maximum=0.5, Evaluation Type=::NewEnumerator1)
    )
    Module.Write Color=true
    Module.Write Lifetime=true
    Module.Write Mass=true
    Module.Write Material Random=true
    Module.Write Position=false
    Module.Write Scale=true
    Module.Write SpriteRotation=false
    Module.Write SpriteSize=false
    Position=<-2 0 0>
   )
   &DynamicMaterialParameters(
    Index 0 Param 3=3
    Index 0 Param 4=1.3752
    Module.Param0WriteEnabled=false
    Module.Param1WriteEnabled=false
    Module.Param2WriteEnabled=true
    Module.Param3WriteEnabled=true
   ) {module=DynamicMaterialParameters001}
  ]
  particleUpdateScript [
   &ParticleState()
   &DynamicMaterialParameters(
    Index 0 Param 1=&FloatFromCurve(FloatCurve=DataInterface() {propertyValues {Curve {keys [{arriveTangent=3.7225, time=0.48, value=0}, {interpMode=RCIM_Cubic, leaveTangent=5.9384, time=0.9673, value=0.5}]}}})
    Module.Param0WriteEnabled=true
    Module.Param1WriteEnabled=true
    Module.Param2WriteEnabled=false
    Module.Param3WriteEnabled=false
   )
   &Color(Color=&LerpLinearColors(StartColor=$User.Color_01, EndColor=$User.Color_02, LerpFactor=$Particles.NormalizedAge))
   &ScaleColor(
    Color Value To Scale=$Particles.Color
    Scale Alpha=&FloatFromCurve(FloatCurve=DataInterface() {propertyValues {Curve {keys [<at 0 1>,<at 1 0>]}}})
    Scale RGB=&VectorFromFloat(Value=20)
    ScaleA=true
    ScaleRGB=true
   )
  ]
  rendererClasses [&NiagaraMeshRendererProperties]
  renderers [{rendererClass=&NiagaraMeshRendererProperties, rendererIndex=0}]
 }
 Energy_ring {
  emitterUpdateScript [
   &EmitterState(Life Cycle Mode=::NewEnumerator1, Loop Behavior=::NewEnumerator1, Loop Duration=1, UseLoopDelay=false)
   &SpawnBurst_Instantaneous(Spawn Count=1, Use Spawn Probability=false)
  ]
  particleSpawnScript [
   /Niagara/Modules/Spawn/Initialization/V2/InitializeParticle(
    Color=<rgba 1 1 1 2>
    Lifetime=0.5
    Mesh Scale=<0.05 0.1 0.1>
    Mesh Scale Mode=::NewEnumerator4
    UsePositionOffset=false
    Write Material Random=true
   ) {module=InitializeParticle001}
   &InitialMeshOrientation(Orientation Vector=<0 0 1>, Rotation=<0 0.25 0>, Use Orientation Vector=true, Use Rotation Vector=true) {module=InitialMeshOrientation001}
   &SystemLocation(MaskSpawnGroup=false, Offset=<7.5 0 0>)
   &DynamicMaterialParameters(
    Index 0 Param 1=2
    Index 0 Param 2=1
    Module.Param0WriteEnabled=false
    Module.Param1WriteEnabled=true
    Module.Param2WriteEnabled=false
    Module.Param3WriteEnabled=false
   ) {module=DynamicMaterialParameters001}
  ]
  particleUpdateScript [
   &ParticleState()
   &ScaleColor(
    Scale Alpha=&Multiply_Float(
     A=&FloatFromCurve(
      FloatCurve=DataInterface() {propertyValues {Curve {keys [
          {arriveTangent=1.25, interpMode=RCIM_Cubic, time=0.05, value=1}
          {arriveTangent=-1.6178, interpMode=RCIM_Cubic, leaveTangent=-1.6178, tangentMode=RCTM_User, time=0.6, value=0.75}
          {arriveTangent=-0.5181, interpMode=RCIM_Cubic, leaveTangent=-0.5181, tangentMode=RCTM_User, time=1, value=0}
         ]}}}
      Scale Curve=2
     )
     B=&FNTimeOfDayScalar(Night Scale=1, Morning Scale=1, Evening Scale=1)
    )
    Scale RGB=&MakeVectorFromLinearColorRGB(Color=&ColorFromCurve(ColorCurve=DataInterface() {propertyValues {
        AlphaCurve {keys []}
        BlueCurve {keys [
          <at 0.1273 1.3018 RCIM_Cubic>
          {arriveTangent=-0.4487, interpMode=RCIM_Cubic, leaveTangent=-0.4487, time=0.2995, value=5}
          <at 0.514 1>
          <at 0.8 1>
         ]}
        GreenCurve {keys [
          <at 0.1273 2 RCIM_Cubic>
          {arriveTangent=-2.9732, interpMode=RCIM_Cubic, leaveTangent=-2.9732, time=0.2995, value=0.5}
          <at 0.514 0>
          <at 0.8 0>
         ]}
        RedCurve {keys [
          <at 0.1273 0.0958 RCIM_Cubic>
          {arriveTangent=-0.0681, interpMode=RCIM_Cubic, leaveTangent=-0.0681, time=0.2995, value=4.25}
          <at 0.514 0.6833>
          <at 0.8 0.05>
         ]}
       }}))
    ScaleA=true
    ScaleRGB=true
   )
   &ScaleMeshSize(Scale Factor=&VectorFromCurve(VectorCurve=DataInterface() {propertyValues {
       XCurve {keys [
         {arriveTangent=-3.0456, interpMode=RCIM_Cubic, leaveTangent=-3.0456, tangentMode=RCTM_User, time=0, value=2}
         {interpMode=RCIM_Cubic, tangentMode=RCTM_User, time=0.9, value=0.75}
        ]}
       YCurve {keys [
         {arriveTangent=3.1974, interpMode=RCIM_Cubic, leaveTangent=3.1974, tangentMode=RCTM_User, time=0, value=0}
         {arriveTangent=-0.0405, interpMode=RCIM_Cubic, leaveTangent=-0.0405, tangentMode=RCTM_User, time=1, value=1}
        ]}
       ZCurve {keys [
         {arriveTangent=3.1974, interpMode=RCIM_Cubic, leaveTangent=3.1974, tangentMode=RCTM_User, time=0, value=0}
         {arriveTangent=-0.0405, interpMode=RCIM_Cubic, leaveTangent=-0.0405, tangentMode=RCTM_User, time=1, value=1}
        ]}
      }}))
   &DynamicMaterialParameters(
    Index 0 Param 1=&FloatFromCurve(FloatCurve=DataInterface() {propertyValues {Curve {keys [<at 0 0.75>,<at 0.6 2>]}}})
    Module.Param0WriteEnabled=true
    Module.Param1WriteEnabled=false
    Module.Param2WriteEnabled=false
    Module.Param3WriteEnabled=false
   )
  ]
  rendererClasses [&NiagaraMeshRendererProperties]
  renderers [{rendererClass=&NiagaraMeshRendererProperties, rendererIndex=0}]
 }
 Flame001 {
  emitterUpdateScript [
   &EmitterState(Life Cycle Mode=::NewEnumerator1, Loop Behavior=::NewEnumerator1, Loop Duration=1, UseLoopDelay=true)
   &SpawnBurst_Instantaneous(Spawn Count=8, Use Spawn Probability=false)
  ]
  particleSpawnScript [
   &InitializeParticle(
    Lifetime Max=0.4
    Lifetime Min=0.2
    Lifetime Mode=::NewEnumerator1
    Mesh Scale Mode=::NewEnumerator4
    Sprite Rotation Mode=::NewEnumerator2
    Sprite Size Mode=::NewEnumerator1
    Uniform Sprite Size Max=75
    Uniform Sprite Size Min=60
    UsePositionOffset=false
    Write Material Random=true
   ) {module=InitializeParticle001}
   &DynamicMaterialParameters(
    Index 0 Param 2=&RandomRangeFloat(Minimum=0, Maximum=10, Evaluation Type=::NewEnumerator1)
    Module.Param0WriteEnabled=false
    Module.Param1WriteEnabled=true
    Module.Param2WriteEnabled=false
    Module.Param3WriteEnabled=false
   ) {module=DynamicMaterialParameters001}
   &AddVelocity(Velocity=&RandomRangeVector(Minimum=<20 0 0>, Maximum=<40 0 0>))
   &AddVelocityInCone(
    Cone Angle=0
    Module.Use Velocity Falloff On Cone Axis=false
    Velocity Distribution Along Cone Axis=0
    Velocity Strength=&RandomRangeFloat(Minimum=40, Maximum=80)
   ) {enabled=false}
   &SystemLocation(MaskSpawnGroup=false, Offset=&RandomRangeVector(Minimum=<5 0 0>, Maximum=<30 0 0>))
  ]
  particleUpdateScript [
   &ParticleState()
   &Color(Color=&MakeLinearColorFromVectorAndFloat(
     "Vector (RGB)"=&MakeVectorFromLinearColorRGB(Color=&ColorFromCurve(ColorCurve=DataInterface() {propertyValues {
         AlphaCurve {keys [<at 0 1 RCIM_Cubic>,<at 1 1 RCIM_Cubic>]}
         BlueCurve {keys [{arriveTangent=-42.695, interpMode=RCIM_Cubic, leaveTangent=-42.695, tangentMode=RCTM_User, time=0, value=25}, <at 0.5945 4 RCIM_Cubic>]}
         GreenCurve {keys [
           {arriveTangent=-29.9613, interpMode=RCIM_Cubic, leaveTangent=-29.9613, tangentMode=RCTM_User, time=0, value=6.25}
           <at 0.5945 0.6 RCIM_Cubic>
          ]}
         RedCurve {keys [<at 0 8.125 RCIM_Cubic>,<at 0.5945 2.0167 RCIM_Cubic>]}
        }}))
     "Float (Alpha)"=&ClampFloat(Float=&FloatFromCurve(FloatCurve=DataInterface() {propertyValues {Curve {keys [
           {arriveTangent=-5.3615, interpMode=RCIM_Cubic, leaveTangent=-5.3615, tangentMode=RCTM_User, time=0, value=1}
           <at 0.2022 0.2988 RCIM_Cubic>
           {arriveTangent=2.6127, interpMode=RCIM_Cubic, leaveTangent=2.6127, tangentMode=RCTM_User, time=1, value=1}
          ]}}}))
    ))
   &DynamicMaterialParameters(
    Index 0 Param 1=&FloatFromCurve(FloatCurve=DataInterface() {propertyValues {Curve {keys [{arriveTangent=0.7395, interpMode=RCIM_Cubic, time=0, value=0.0108}, {interpMode=RCIM_Cubic, leaveTangent=0.1757, time=1, value=0.4945}]}}})
    Index 0 Param 3=1
    Index 0 Param 4=1
    Module.Param0WriteEnabled=true
    Module.Param1WriteEnabled=false
    Module.Param2WriteEnabled=true
    Module.Param3WriteEnabled=true
   )
   &ScaleSpriteSize(Uniform Curve Sprite Scale=DataInterface() {propertyValues {Curve {keys [{tangentMode=RCTM_User, time=0, value=0.1}, {tangentMode=RCTM_User, time=1, value=1.1}]}}})
   &Drag(Use Linear Drag=true, Use Rotational Drag=false)
   &SolveForcesAndVelocity(Clamp Velocity=false, Limit Acceleration=false)
  ]
  rendererClasses [&NiagaraSpriteRendererProperties]
  renderers [{rendererClass=&NiagaraSpriteRendererProperties}]
 }
 GlowStart {
  emitterUpdateScript [
   &EmitterState()
   &SpawnBurst_Instantaneous(Spawn Count=1, Use Spawn Probability=false)
  ]
  particleSpawnScript [
   /Niagara/Modules/Spawn/Initialization/InitializeParticle(
    Color=$User.Color_01
    Lifetime=&RandomRangeFloat(Minimum=0.2, Maximum=0.25, Evaluation Type=::NewEnumerator1)
    Mesh Scale=&VectorFromFloat(Value=3)
    Module.Write Color=true
    Module.Write Lifetime=true
    Module.Write Mass=true
    Module.Write Material Random=true
    Module.Write Position=true
    Module.Write Scale=true
    Module.Write SpriteRotation=false
    Module.Write SpriteSize=false
    Position=&SimulationPosition()
   ) {enabled=false}
   /Niagara/Modules/Spawn/Initialization/V2/InitializeParticle(
    Color=<rgba 0.97 0.1 1 1>
    Lifetime Max=0.25
    Lifetime Min=0.2
    Lifetime Mode=::NewEnumerator1
    Mesh Scale Mode=::NewEnumerator3
    Mesh Uniform Scale=0.35
    Position Offset=<8 0 0>
    UsePositionOffset=true
    Write Material Random=true
   ) {module=InitializeParticle001}
  ]
  particleUpdateScript [
   &ParticleState()
   &DynamicMaterialParameters(
    Index 0 Param 1=&FloatFromCurve(FloatCurve=DataInterface() {propertyValues {Curve {keys [<at 0 1>,<at 1 1.5>]}}})
    Index 0 Param 2=30
    Index 0 Param 3=10
    Index 0 Param 4=50
    Module.Param0WriteEnabled=true
    Module.Param1WriteEnabled=true
    Module.Param2WriteEnabled=true
    Module.Param3WriteEnabled=true
   )
   &ScaleColor(
    Scale Alpha=&FloatFromCurve(FloatCurve=DataInterface() {propertyValues {Curve {keys [<at 0 0 RCIM_Cubic>,<at 0.2 1 RCIM_Cubic>,<at 1 0 RCIM_Cubic>]}}}, Scale Curve=0.01)
    Scale RGB=&VectorFromFloat(Value=5)
    ScaleA=true
    ScaleRGB=true
   )
   &CameraOffset(Camera Offset Amount=50)
   &FN_Light_Attributes(
    Color=&LerpLinearColors(
     StartColor=<rgba 0 1 0.8115 1>
     EndColor=<rgba 0.9 0 1 1>
     LerpFactor=&FloatFromCurve(FloatCurve=DataInterface() {propertyValues {Curve {keys [{arriveTangent=2.781, interpMode=RCIM_Cubic, leaveTangent=2.781, tangentMode=RCTM_User, time=0, value=0}, <at 1 1>]}}})
    )
    Create Color Attribute=true
    Create Exponent Attribute=true
    Create Radius Attribute=true
    Exponent=1
    Radius=50
    Scale Color=&FloatFromCurve(FloatCurve=DataInterface() {propertyValues {Curve {keys [<at 0 0 RCIM_Cubic>,<at 0.2 1 RCIM_Cubic>,<at 1 0 RCIM_Cubic>]}}}, Scale Curve=10)
   )
  ]
  rendererClasses [&NiagaraMeshRendererProperties, &NiagaraLightRendererProperties]
  renderers [
   {rendererClass=&NiagaraMeshRendererProperties, rendererIndex=0}
   {rendererClass=&NiagaraLightRendererProperties, rendererIndex=1}
  ]
 }
 MFlashHot {
  emitterUpdateScript [
   &EmitterState(Life Cycle Mode=::NewEnumerator1, Loop Behavior=::NewEnumerator1, Loop Duration=1, UseLoopDelay=false)
   &SpawnBurst_Instantaneous(Spawn Count=1, Use Spawn Probability=false)
  ]
  particleSpawnScript [
   &InitializeParticle(Lifetime=0.2, Mesh Scale=<0.8 0.6 0.6>, Mesh Scale Mode=::NewEnumerator4, UsePositionOffset=false, Write Material Random=true) {module=InitializeParticle001}
   &Color(Color=&MakeLinearColorFromVectorAndFloat("Vector (RGB)"=&MakeVectorFromLinearColorRGB(Color=<rgba 2 2 2 1>), "Float (Alpha)"=1))
   &AddVelocity(Velocity=<25 0 0>) {enabled=false}
   &SystemLocation(MaskSpawnGroup=false, Offset=<5.5 0 0>)
   &InitialMeshOrientation(
    Rotation=&RandomRangeVector(Minimum=<0 0 0>, Maximum=<1 0 0>)
    Use Orientation Vector=true
    Use Rotation Vector=true
   )
  ]
  particleUpdateScript [
   &SolveForcesAndVelocity(Clamp Velocity=false, Limit Acceleration=false)
   &ParticleState()
   &DynamicMaterialParameters(Module.Param0WriteEnabled=true, Module.Param1WriteEnabled=true, Module.Param2WriteEnabled=true, Module.Param3WriteEnabled=true)
   &ScaleColor(
    Scale RGB=&MakeVectorFromLinearColorRGB(Color=&ColorFromCurve(ColorCurve=DataInterface() {propertyValues {
        AlphaCurve {keys [<at 0 1 RCIM_Cubic>]}
        BlueCurve {keys [<at 0 0.9667 RCIM_Cubic>,<at 0.8 1>]}
        GreenCurve {keys [<at 0 1 RCIM_Cubic>,<at 0.8 0>]}
        RedCurve {keys [<at 0 0 RCIM_Cubic>,<at 0.8 0.5667>]}
       }}))
    ScaleA=true
    ScaleRGB=true
   )
   &ScaleMeshSize(Scale Factor=&VectorFromCurve(VectorCurve=DataInterface() {propertyValues {
       XCurve {keys [
         {arriveTangent=3.0736, interpMode=RCIM_Cubic, leaveTangent=3.0736, tangentMode=RCTM_User, time=0, value=0.2}
         {arriveTangent=0.494, interpMode=RCIM_Cubic, leaveTangent=0.494, tangentMode=RCTM_User, time=0.3, value=0.95}
         {arriveTangent=0.0804, interpMode=RCIM_Cubic, leaveTangent=0.0804, tangentMode=RCTM_User, time=1, value=1.15}
        ]}
       YCurve {keys [
         {arriveTangent=1.1344, interpMode=RCIM_Cubic, leaveTangent=1.1344, tangentMode=RCTM_User, time=0, value=0.4}
         {arriveTangent=-0.1563, interpMode=RCIM_Cubic, leaveTangent=-0.1563, tangentMode=RCTM_User, time=0.3, value=0.65}
         {arriveTangent=0.1176, interpMode=RCIM_Cubic, leaveTangent=0.1176, tangentMode=RCTM_User, time=1, value=0}
        ]}
       ZCurve {keys [
         {arriveTangent=1.1344, interpMode=RCIM_Cubic, leaveTangent=1.1344, tangentMode=RCTM_User, time=0, value=0.4}
         {arriveTangent=-0.1563, interpMode=RCIM_Cubic, leaveTangent=-0.1563, tangentMode=RCTM_User, time=0.3, value=0.65}
         {arriveTangent=0.4791, interpMode=RCIM_Cubic, leaveTangent=0.4791, tangentMode=RCTM_User, time=1, value=0}
        ]}
      }}))
  ]
  rendererClasses [&NiagaraMeshRendererProperties]
  renderers [{rendererClass=&NiagaraMeshRendererProperties, rendererIndex=0}]
 }
 MFlashSoft {
  emitterUpdateScript [
   &EmitterState(Life Cycle Mode=::NewEnumerator1, Loop Behavior=::NewEnumerator1, Loop Duration=1, UseLoopDelay=false)
   &SpawnBurst_Instantaneous(Spawn Count=1, Use Spawn Probability=false)
  ]
  particleSpawnScript [
   &InitializeParticle(Lifetime=0.15, Mesh Scale=<0.9 1.5 1.5>, Mesh Scale Mode=::NewEnumerator4, UsePositionOffset=false, Write Material Random=true) {module=InitializeParticle001}
   &Color(Color=&MakeLinearColorFromVectorAndFloat("Vector (RGB)"=<0.5 0.5 0.5>, "Float (Alpha)"=&ClampFloat(Float=0.65)))
   &AddVelocity(Velocity=<25 0 0>) {enabled=false}
   &SystemLocation(MaskSpawnGroup=false, Offset=<6 0 0>)
   &InitialMeshOrientation(
    Rotation=&RandomRangeVector(Minimum=<0 0 0>, Maximum=<1 0 0>)
    Use Orientation Vector=true
    Use Rotation Vector=true
   )
  ]
  particleUpdateScript [
   &SolveForcesAndVelocity(Clamp Velocity=false, Limit Acceleration=false)
   &ParticleState()
   &DynamicMaterialParameters(Module.Param0WriteEnabled=true, Module.Param1WriteEnabled=true, Module.Param2WriteEnabled=true, Module.Param3WriteEnabled=true)
   &ScaleColor(
    Scale RGB=&MakeVectorFromLinearColorRGB(Color=&ColorFromCurve(ColorCurve=DataInterface() {propertyValues {
        AlphaCurve {keys [<at 0 1 RCIM_Cubic>]}
        BlueCurve {keys [<at 0 0.9667 RCIM_Cubic>,<at 0.8 1>]}
        GreenCurve {keys [<at 0 1 RCIM_Cubic>,<at 0.8 0>]}
        RedCurve {keys [<at 0 0 RCIM_Cubic>,<at 0.8 0.5667>]}
       }}))
    ScaleA=true
    ScaleRGB=true
   ) {module=ScaleColor001}
   &ScaleMeshSize(Scale Factor=&VectorFromCurve(VectorCurve=DataInterface() {propertyValues {
       XCurve {keys [
         {arriveTangent=3.0736, interpMode=RCIM_Cubic, leaveTangent=3.0736, tangentMode=RCTM_User, time=0, value=0.2}
         {arriveTangent=0.178, interpMode=RCIM_Cubic, leaveTangent=0.178, tangentMode=RCTM_User, time=0.3, value=0.65}
         {arriveTangent=-0.1748, interpMode=RCIM_Cubic, leaveTangent=-0.1748, tangentMode=RCTM_User, time=1, value=0.8}
        ]}
       YCurve {keys [
         {arriveTangent=0.6326, interpMode=RCIM_Cubic, leaveTangent=0.6326, tangentMode=RCTM_User, time=0, value=0.4}
         {arriveTangent=-0.1563, interpMode=RCIM_Cubic, leaveTangent=-0.1563, tangentMode=RCTM_User, time=0.3, value=0.65}
         {arriveTangent=-0.5198, interpMode=RCIM_Cubic, leaveTangent=-0.5198, tangentMode=RCTM_User, time=1, value=0}
        ]}
       ZCurve {keys [
         {arriveTangent=0.6326, interpMode=RCIM_Cubic, leaveTangent=0.6326, tangentMode=RCTM_User, time=0, value=0.4}
         {arriveTangent=-0.1563, interpMode=RCIM_Cubic, leaveTangent=-0.1563, tangentMode=RCTM_User, time=0.3, value=0.65}
         {arriveTangent=-0.5198, interpMode=RCIM_Cubic, leaveTangent=-0.5198, tangentMode=RCTM_User, time=1, value=0}
        ]}
      }}))
  ]
  rendererClasses [&NiagaraMeshRendererProperties]
  renderers [{rendererClass=&NiagaraMeshRendererProperties, rendererIndex=0}]
 }
 MagneticCylinders {
  emitterUpdateScript [
   &EmitterState()
   &SpawnBurst_Instantaneous(Spawn Count=1, Use Spawn Probability=false)
   &SpawnBurst_Instantaneous(Spawn Count=2, Spawn Time=0.05, Use Spawn Probability=false) {enabled=false, module=SpawnBurst_Instantaneous001}
  ]
  particleSpawnScript [
   &FN_SetSpawnGroup(Number of Spawn Groups=3)
   &InitializeParticle(
    Color=$User.Color_01
    Lifetime=0.2
    Mesh Scale=&SelectVectorFromArray(
     Direct Array Index=&SpawnGroup()
     Vector Selection Array=DataInterface() {propertyValues {FloatData [<0.75 0.5 0.5>, <0.5 0.8 0.6>, <0.5 0.8 0.6>]}}
    )
    Mesh Scale Mode=::NewEnumerator4
    UsePositionOffset=false
    Write Material Random=true
   )
   &InitialMeshOrientation(
    Orientation Vector=&TransformVector(
     Vector=&SelectVectorFromArray(
      Direct Array Index=&SpawnGroup()
      Vector Selection Array=DataInterface() {propertyValues {FloatData [<1 0 0>, <0.5 0.3 0>, <0.5 -0.3 0>]}}
     )
     Destination Space=::Simulation
    )
    Use Orientation Vector=true
    Use Rotation Vector=false
   ) {enabled=false}
   /SetVariables(Particles.MeshOrientation=&MakeQuatFromAxisAngle(Axis=<0 0 0>, AngleInDegrees=90)) {isSetParameters=true}
  ]
  particleUpdateScript [
   &ParticleState()
   &ScaleMeshSize(Scale Factor=&VectorFromCurve(VectorCurve=DataInterface() {propertyValues {
       XCurve {keys [
         {arriveTangent=1.0339, interpMode=RCIM_Cubic, leaveTangent=1.0339, tangentMode=RCTM_User, time=0, value=0.4}
         {interpMode=RCIM_Cubic, tangentMode=RCTM_User, time=1, value=1}
        ]}
       YCurve {keys [
         {arriveTangent=15.572, interpMode=RCIM_Cubic, leaveTangent=15.572, tangentMode=RCTM_User, time=0, value=0}
         {arriveTangent=-1.8829, interpMode=RCIM_Cubic, leaveTangent=-1.8829, tangentMode=RCTM_User, time=0.15, value=1.5}
         {interpMode=RCIM_Cubic, leaveTangent=-1.8864, time=1, value=0.2}
        ]}
       ZCurve {keys [
         {arriveTangent=15.572, interpMode=RCIM_Cubic, leaveTangent=15.572, tangentMode=RCTM_User, time=0, value=0}
         {arriveTangent=-1.8829, interpMode=RCIM_Cubic, leaveTangent=-1.8829, tangentMode=RCTM_User, time=0.15, value=1.5}
         {interpMode=RCIM_Cubic, leaveTangent=-1.8864, time=1, value=0.2}
        ]}
      }}))
   &DynamicMaterialParameters(
    Index 0 Param 1=&MaskFloatBySpawnGroup(
     Masked Float=0
     Passthrough Float=&FloatFromCurve(FloatCurve=DataInterface() {propertyValues {Curve {keys [
          {arriveTangent=29.814, interpMode=RCIM_Cubic, leaveTangent=29.814, tangentMode=RCTM_User, time=0, value=0}
          {arriveTangent=-2.5249, interpMode=RCIM_Cubic, leaveTangent=-2.5249, tangentMode=RCTM_User, time=0.2, value=2}
          {arriveTangent=-0.0405, interpMode=RCIM_Cubic, leaveTangent=-0.0405, tangentMode=RCTM_User, time=1, value=0.5}
         ]}}})
    )
    Index 0 Param 2=&FloatFromCurve(
     FloatCurve=DataInterface() {propertyValues {Curve {keys [
         {arriveTangent=1.5778, interpMode=RCIM_Cubic, leaveTangent=1.5778, tangentMode=RCTM_User, time=0, value=-0.1}
         {arriveTangent=0.4814, interpMode=RCIM_Cubic, leaveTangent=0.4814, tangentMode=RCTM_User, time=1, value=1}
        ]}}}
     Scale Curve=1.25
    )
    Index 0 Param 3=&FloatFromCurve(FloatCurve=DataInterface() {propertyValues {Curve {keys [
         {interpMode=RCIM_Cubic, tangentMode=RCTM_User, time=0, value=1}
         {arriveTangent=-1.9967, interpMode=RCIM_Cubic, leaveTangent=-1.9967, tangentMode=RCTM_User, time=1, value=0}
        ]}}})
    Index 0 Param 4=&FloatFromCurve(
     FloatCurve=DataInterface() {propertyValues {Curve {keys [
         {arriveTangent=2.2707, interpMode=RCIM_Cubic, leaveTangent=2.2707, tangentMode=RCTM_User, time=0, value=0}
         {arriveTangent=-0.0405, interpMode=RCIM_Cubic, leaveTangent=-0.0405, tangentMode=RCTM_User, time=1, value=1}
        ]}}}
     Scale Curve=2
    )
    Module.Param0WriteEnabled=true
    Module.Param1WriteEnabled=true
    Module.Param2WriteEnabled=false
    Module.Param3WriteEnabled=true
   )
   &Color(Color=&LerpLinearColors(
     StartColor=$User.Color_01
     EndColor=$User.Color_03
     LerpFactor=&FloatFromCurve(FloatCurve=DataInterface() {propertyValues {Curve {keys [<at 0 0>,<at 0.5 1>]}}})
    ))
   &ScaleColor(
    Color Value To Scale=$Particles.Color
    Scale Alpha=&FloatFromCurve(
     FloatCurve=DataInterface() {propertyValues {Curve {keys [{interpMode=RCIM_Cubic, tangentMode=RCTM_User, time=0, value=3}, {interpMode=RCIM_Cubic, tangentMode=RCTM_User, time=1, value=0}]}}}
     Scale Curve=0.5
    )
    Scale RGB=&VectorFromFloat(Value=&FloatFromCurve(FloatCurve=DataInterface() {propertyValues {Curve {keys [{arriveTangent=-8.1302, interpMode=RCIM_Cubic, leaveTangent=-8.1302, tangentMode=RCTM_User, time=0, value=5}, <at 1 1>]}}}))
    ScaleA=true
    ScaleRGB=true
   )
  ]
  rendererClasses [&NiagaraMeshRendererProperties]
  renderers [{rendererClass=&NiagaraMeshRendererProperties, rendererIndex=0}]
 }
 MuzzleFlare {
  bEnabled=false
  emitterUpdateScript [
   &EmitterState()
   &SpawnBurst_Instantaneous(Spawn Count=3, Use Spawn Probability=false)
  ]
  particleSpawnScript [
   /SetVariables(Particles.MeshOrientation=&MakeQuatFromAxisAngle(Axis=<0 0 0>, AngleInDegrees=90)) {isSetParameters=true}
   /Niagara/Modules/Spawn/Initialization/InitializeParticle(
    Color=$User.Color_01
    Lifetime=0.25
    Mesh Scale=&RandomRangeVector(Minimum=<2.2 0.8 0.8>, Maximum=<2.5 1 1>, Evaluation Type=::NewEnumerator1)
    Module.Write Color=true
    Module.Write Lifetime=true
    Module.Write Mass=true
    Module.Write Material Random=true
    Module.Write Position=true
    Module.Write Scale=true
    Module.Write SpriteRotation=false
    Module.Write SpriteSize=false
    Position=&Add_Vector(A=&SimulationPosition(), B=<7 0 0>)
   )
  ]
  particleUpdateScript [
   &ParticleState()
   &ScaleColor(
    Scale Alpha=&FloatFromCurve(FloatCurve=DataInterface() {propertyValues {Curve {keys [
         {interpMode=RCIM_Cubic, tangentMode=RCTM_User, time=0, value=0}
         {interpMode=RCIM_Cubic, tangentMode=RCTM_User, time=0.5, value=1}
         {tangentMode=RCTM_User, time=1, value=0}
        ]}}})
    Scale RGB=&VectorFromFloat(Value=&FloatFromCurve(
      FloatCurve=DataInterface() {propertyValues {Curve {keys [
          {arriveTangent=9.0725, interpMode=RCIM_Cubic, leaveTangent=9.0725, tangentMode=RCTM_User, time=0, value=0}
          {arriveTangent=-2.5249, interpMode=RCIM_Cubic, leaveTangent=-2.5249, tangentMode=RCTM_User, time=0.1408, value=0.9847}
          {arriveTangent=-0.0405, interpMode=RCIM_Cubic, leaveTangent=-0.0405, tangentMode=RCTM_User, time=1, value=0}
         ]}}}
      Scale Curve=10
     ))
    ScaleA=true
    ScaleRGB=true
   )
   &DynamicMaterialParameters(
    Index 0 Param 2=1
    Index 0 Param 3=1
    Index 0 Param 4=1
    Module.Param0WriteEnabled=true
    Module.Param1WriteEnabled=true
    Module.Param2WriteEnabled=true
    Module.Param3WriteEnabled=true
   )
   &ScaleMeshSize(Scale Factor=&MakeVector(
     X=&FloatFromCurve(FloatCurve=DataInterface() {propertyValues {Curve {keys [{interpMode=RCIM_Cubic, tangentMode=RCTM_User, time=0, value=0}, {interpMode=RCIM_Cubic, tangentMode=RCTM_User, time=1, value=1}]}}})
     Y=&FloatFromCurve(FloatCurve=DataInterface() {propertyValues {Curve {keys [{interpMode=RCIM_Cubic, tangentMode=RCTM_User, time=0, value=1}, {interpMode=RCIM_Cubic, tangentMode=RCTM_User, time=1, value=0}]}}})
     Z=&FloatFromCurve(FloatCurve=DataInterface() {propertyValues {Curve {keys [{interpMode=RCIM_Cubic, tangentMode=RCTM_User, time=0, value=1}, {interpMode=RCIM_Cubic, tangentMode=RCTM_User, time=1, value=0}]}}})
    ))
  ]
  rendererClasses [&NiagaraMeshRendererProperties]
  renderers [{rendererClass=&NiagaraMeshRendererProperties, rendererIndex=0}]
 }
 Rings {
  emitterUpdateScript [
   &EmitterState()
   &SpawnRate(
    SpawnRate=&Multiply_Float(
     A=26
     B=&FloatFromCurve(
      FloatCurve=DataInterface() {propertyValues {Curve {keys [{interpMode=RCIM_Cubic, tangentMode=RCTM_User, time=0, value=1}, {interpMode=RCIM_Cubic, tangentMode=RCTM_User, time=0.3, value=0}]}}}
      CurveIndex=$Emitter.NormalizedLoopAge
     )
    )
    Use Spawn Probability=false
   )
   &SpawnBurst_Instantaneous(Spawn Count=1, Use Spawn Probability=false)
  ]
  particleSpawnScript [
   /SetVariables(Particles.MeshOrientation=&MakeQuatFromAxisAngle(Axis=<0 1 0>, AngleInDegrees=90)) {isSetParameters=true}
   &InitializeParticle(
    Lifetime=&RandomRangeFloat(Minimum=0.2, Maximum=0.2, Evaluation Type=::NewEnumerator1)
    Mesh Scale=&VectorFromFloat(Value=0.07)
    Module.Write Color=true
    Module.Write Lifetime=true
    Module.Write Mass=true
    Module.Write Material Random=true
    Module.Write Position=true
    Module.Write Scale=true
    Module.Write SpriteRotation=false
    Module.Write SpriteSize=false
    Position=<0 0 0>
   )
   &AddVelocity(Velocity=<200 0 0>)
  ]
  particleUpdateScript [
   &ParticleState()
   &Color(Color=&LerpLinearColors(StartColor=$User.Color_01, EndColor=$User.Color_03, LerpFactor=$Particles.NormalizedAge))
   &ScaleColor(
    Color Value To Scale=$Particles.Color
    Scale Alpha=&FloatFromCurve(
     FloatCurve=DataInterface() {propertyValues {Curve {keys [
         {arriveTangent=15.9921, interpMode=RCIM_Cubic, time=0, value=0}
         {arriveTangent=11.3331, interpMode=RCIM_Cubic, leaveTangent=11.3331, tangentMode=RCTM_User, time=0.1021, value=0.9938}
         {arriveTangent=-8.8252, leaveTangent=-8.8252, time=1, value=0}
        ]}}}
     Scale Curve=0.2
    )
    Scale RGB=&VectorFromFloat(Value=3)
    ScaleA=true
    ScaleRGB=true
   )
   &SolveForcesAndVelocity(Clamp Velocity=false, Limit Acceleration=false)
   &DynamicMaterialParameters(
    Index 0 Param 1=&FloatFromCurve(
     FloatCurve=DataInterface() {propertyValues {Curve {keys [
         {arriveTangent=9.0725, interpMode=RCIM_Cubic, leaveTangent=9.0725, tangentMode=RCTM_User, time=0, value=0}
         {arriveTangent=0.1068, interpMode=RCIM_Cubic, leaveTangent=0.1068, tangentMode=RCTM_User, time=0.1515, value=0.9243}
         <at 0.7 1>
         {arriveTangent=-0.0405, interpMode=RCIM_Cubic, leaveTangent=-0.0405, tangentMode=RCTM_User, time=1, value=0}
        ]}}}
     Scale Curve=0.1
    )
    Index 0 Param 2=1
    Module.Param0WriteEnabled=true
    Module.Param1WriteEnabled=true
    Module.Param2WriteEnabled=true
    Module.Param3WriteEnabled=true
   )
   &ScaleMeshSize(Scale Factor=&VectorFromFloat(Value=&FloatFromCurve(FloatCurve=DataInterface() {propertyValues {Curve {keys [{interpMode=RCIM_Cubic, tangentMode=RCTM_User, time=0, value=1}, {interpMode=RCIM_Cubic, tangentMode=RCTM_User, time=1, value=1.7}]}}})))
  ]
  rendererClasses [&NiagaraMeshRendererProperties]
  renderers [{rendererClass=&NiagaraMeshRendererProperties, rendererIndex=0}]
 }
 Sparks_Main {
  emitterUpdateScript [
   &EmitterState(Life Cycle Mode=::NewEnumerator1, Loop Behavior=::NewEnumerator1, Loop Duration=1, UseLoopDelay=false)
   &SpawnBurst_Instantaneous(Spawn Count=9, Use Spawn Probability=false) {module=SpawnBurst_Instantaneous001}
  ]
  particleSpawnScript [
   /Niagara/Modules/Spawn/Initialization/V2/InitializeParticle(
    Color=<rgba 5 5 5 1>
    Lifetime Max=0.2
    Lifetime Min=0.15
    Lifetime Mode=::NewEnumerator1
    Sprite Size Max=<2 3>
    Sprite Size Min=<1 1>
    Sprite Size Mode=::NewEnumerator2
    UsePositionOffset=false
    Write Material Random=true
   ) {module=InitializeParticle001}
   &SphereLocation(
    Hemisphere X=true
    MaskSpawnGroup=false
    Override Seed=false
    Sphere Radius=3
    Surface Only=false
    UseNonUniformScale=false
    UseOffset=false
    UseSphereOrientation=false
   )
   &AddVelocityInCone(
    Cone Angle=25
    Module.Use Velocity Falloff On Cone Axis=false
    Velocity Strength=&RandomRangeFloat(Minimum=200, Maximum=700)
   )
  ]
  particleUpdateScript [
   &ParticleState()
   &Drag(
    Drag=&FloatFromCurve(FloatCurve=DataInterface() {propertyValues {Curve {keys [{interpMode=RCIM_Cubic, tangentMode=RCTM_Break, time=0, value=2.5}, {interpMode=RCIM_Cubic, tangentMode=RCTM_Break, time=1, value=4}]}}})
    Use Linear Drag=true
    Use Rotational Drag=false
   )
   &GravityForce(Gravity=&MakeVector(X=0, Y=0, Z=&RandomRangeFloat(Minimum=-150, Maximum=-500))) {enabled=false}
   &SolveForcesAndVelocity(Clamp Velocity=false, Limit Acceleration=false)
   &ScaleSpriteSizeBySpeed(Max Scale Factor=<1 10>, Min Scale Factor=<1 5>)
   &ScaleSpriteSize(Scale Factor=&MakeVector2D(
     X=&FloatFromCurve(FloatCurve=DataInterface() {propertyValues {Curve {keys [<at 0.7 1>,<at 1 0.5>]}}})
     Y=&FloatFromCurve(FloatCurve=DataInterface() {propertyValues {Curve {keys [<at 0 2>,<at 0.1 1>,<at 0.8 0.5>,<at 1 0>]}}})
    ))
   &ScaleColor(
    Scale Alpha=&FloatFromCurve(
     FloatCurve=DataInterface() {propertyValues {Curve {keys [
         {interpMode=RCIM_Cubic, tangentMode=RCTM_Break, time=0, value=1}
         {arriveTangent=-2.6488, interpMode=RCIM_Cubic, tangentMode=RCTM_Break, time=1, value=0}
        ]}}}
     Scale Curve=&RandomRangeFloat(Minimum=1, Maximum=1.5)
    )
    Scale RGB=&MakeVectorFromLinearColorRGB(Color=<rgba 1 0 0.9833 1>)
    ScaleA=true
    ScaleRGB=true
   )
  ]
  rendererClasses [&NiagaraSpriteRendererProperties]
  renderers [{rendererClass=&NiagaraSpriteRendererProperties}]
 }
 Streak {
  bEnabled=false
  emitterUpdateScript [
   &EmitterState(Life Cycle Mode=::NewEnumerator1, Loop Behavior=::NewEnumerator1, Loop Duration=0.1, UseLoopDelay=true)
   &SpawnBurst_Instantaneous(Spawn Count=&RandomRangeInt(Minimum=1, Maximum=1), Use Spawn Probability=false)
  ]
  particleSpawnScript [
   /SetVariables(Particles.MeshOrientation=&MakeQuatFromAxisAngle(
     Axis=&RandomRangeVector(Minimum=<-1 0 0>, Maximum=<1 0 0>, Evaluation Type=::NewEnumerator1)
     AngleInDegrees=90
    )) {isSetParameters=true}
   /Niagara/Modules/Spawn/Initialization/InitializeParticle(
    Lifetime=0.1
    Mesh Scale=&SplitVector(
     Double Component Vector Channels:ECascadeNiagaraTwoVectorChannels=::NewEnumerator1
     Double Component:NiagaraFloat=&RandomRangeFloat(Minimum=2, Maximum=2.5, Evaluation Type=::NewEnumerator1)
     Single Component:NiagaraFloat=&RandomRangeFloat(Minimum=3, Maximum=3.2, Evaluation Type=::NewEnumerator1)
    )
    Module.Write Color=true
    Module.Write Lifetime=true
    Module.Write Mass=true
    Module.Write Material Random=true
    Module.Write Position=false
    Module.Write Scale=true
    Module.Write SpriteRotation=false
    Module.Write SpriteSize=false
    Position=<2 0 0>
   )
   &DynamicMaterialParameters(
    Index 0 Param 3=3
    Index 0 Param 4=1.3752
    Module.Param0WriteEnabled=false
    Module.Param1WriteEnabled=false
    Module.Param2WriteEnabled=true
    Module.Param3WriteEnabled=true
   ) {module=DynamicMaterialParameters001}
  ]
  particleUpdateScript [
   &ParticleState()
   &DynamicMaterialParameters(
    Index 0 Param 1=&FloatFromCurve(FloatCurve=DataInterface() {propertyValues {Curve {keys [{arriveTangent=3.7225, time=0.48, value=0}, {interpMode=RCIM_Cubic, leaveTangent=5.9384, time=0.9673, value=1}]}}})
    Index 0 Param 2=&FloatFromCurve(FloatCurve=DataInterface() {propertyValues {Curve {keys [<at 0.1507 4.9359 RCIM_Cubic>,<at 1 0 RCIM_Cubic>]}}})
    Module.Param0WriteEnabled=true
    Module.Param1WriteEnabled=true
    Module.Param2WriteEnabled=false
    Module.Param3WriteEnabled=false
   )
   &Color(Color=&LerpLinearColors(StartColor=$User.Color_01, EndColor=$User.Color_02, LerpFactor=$Particles.NormalizedAge)) {module=Color001}
   &ScaleColor(
    Color Value To Scale=$Particles.Color
    Scale Alpha=&FloatFromCurve(FloatCurve=DataInterface() {propertyValues {Curve {keys [<at 0 1>,<at 1 0>]}}})
    Scale RGB=&VectorFromFloat(Value=10)
    ScaleA=true
    ScaleRGB=true
   ) {module=ScaleColor001}
  ]
  rendererClasses [&NiagaraMeshRendererProperties]
  renderers [{rendererClass=&NiagaraMeshRendererProperties, rendererIndex=0}]
 }
}
summary {
 systemName=NS_KnockbackWeapon_Muzzleflash
 emitters [
  {emitterName=GlowStart, rendererClasses [&NiagaraMeshRendererProperties, &NiagaraLightRendererProperties]}
  {emitterName=Rings, rendererClasses [&NiagaraMeshRendererProperties]}
  {bEnabled=false, emitterName=MuzzleFlare, rendererClasses [&NiagaraMeshRendererProperties]}
  {bEnabled=false, emitterName=Core, rendererClasses [&NiagaraMeshRendererProperties]}
  {bEnabled=false, emitterName=Streak, rendererClasses [&NiagaraMeshRendererProperties]}
  {emitterName=BarrelHole, rendererClasses [&NiagaraSpriteRendererProperties]}
  {emitterName=MagneticCylinders, rendererClasses [&NiagaraMeshRendererProperties]}
  {emitterName=Sparks_Main, rendererClasses [&NiagaraSpriteRendererProperties]}
  {emitterName=Energy_ring, rendererClasses [&NiagaraMeshRendererProperties]}
  {emitterName=MFlashSoft, rendererClasses [&NiagaraMeshRendererProperties]}
  {emitterName=MFlashHot, rendererClasses [&NiagaraMeshRendererProperties]}
  {emitterName=Flame001, rendererClasses [&NiagaraSpriteRendererProperties]}
 ]
 userVariables [
  User.Color_01 {type=LinearColor, description="", defaultValue {struct=LinearColor, value=<rgba 0 1 0.8115 1>}}
  User.Color_02 {type=LinearColor, description="", defaultValue {struct=LinearColor, value=<rgba 1 0.0396 0.894 1>}}
  User.Color_03 {type=LinearColor, description="", defaultValue {struct=LinearColor, value=<rgba 1 0.0396 0.894 1>}}
 ]
}
